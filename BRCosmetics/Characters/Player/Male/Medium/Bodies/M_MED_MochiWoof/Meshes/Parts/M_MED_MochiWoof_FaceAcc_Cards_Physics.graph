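# Physics asset for the MochiWoof male cosmetic's cards accessory, covering a belt-hung card holder, four cards and belt side pieces.
# Free constraints chain the belt from the pelvis, branching into a two-bone side strap and the card holder, which has tight angular limits.
# Cards 02–04 get slight single-axis swing from the holder, while card 01 has zero limits.
# Large pelvis and left-thigh capsules stand in for the body.
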PhysicsAsset /BRCosmetics/Characters/Player/Male/Medium/Bodies/M_MED_MochiWoof/Meshes/Parts/M_MED_MochiWoof_FaceAcc_Cards_Physics
bodies {
 base_belt_cardsholder [{shapeType=Capsule, length=0.505, radius=0.505, shapeName=base_belt_cardsholder_capsule, rotation=<rot 0 -90 0>}]
 base_belt_ref [{shapeType=Capsule, length=0.505, radius=0.505, shapeName=base_belt_ref_capsule, rotation=<rot 0 -90 0>}]
 dyn_belt_cardsholder [
  {
   shapeType=Capsule
   length=8.0779
   radius=1.4679
   shapeName=dyn_belt_cardsholder_capsule
   center=<6.62 -0.8136 -1.0795>
   rotation=<rot 65.0617 -53.0196 -48.9085>
  }
  {
   shapeType=Box
   extentX=13.5133
   extentY=1.1996
   extentZ=12.2078
   shapeName=dyn_belt_cardsholder_box
   center=<6.7999 -0.216 -4.417>
   rotation=<rot 50.0001 0 0>
  }
 ]
 dyn_belt_ref [{shapeType=Capsule, length=0.505, radius=0.505, shapeName=dyn_belt_ref_capsule, rotation=<rot 0 -90 0>}]
 dyn_belt_side_01 [{
   shapeType=Capsule
   length=1.0707
   radius=0.9743
   shapeName=dyn_belt_side_01_capsule
   center=<2.9053 -0.0896 -0.5095>
   rotation=<rot 1.0658 84.3736 -88.7296>
  }]
 dyn_belt_side_02 [{
   shapeType=Capsule
   length=3.9973
   radius=0.8275
   shapeName=dyn_belt_side_02_capsule
   center=<2.1632 0.358 -0.0048>
   rotation=<rot 0 95.7323 -81.7535>
  }]
 dyn_cards_01 [{shapeType=Capsule, length=3, radius=1, shapeName=dyn_cards_01_capsule, center=<4 0 0>, rotation=<rot 90 0 0>}]
 dyn_cards_02 [{shapeType=Capsule, length=3, radius=1, shapeName=dyn_cards_02_capsule, center=<7 0 0>, rotation=<rot 90 0 0>}]
 dyn_cards_03 [{shapeType=Capsule, length=3, radius=1, shapeName=dyn_cards_03_capsule, center=<10 0 0>, rotation=<rot 90 0 0>}]
 dyn_cards_04 [{shapeType=Capsule, length=3, radius=1, shapeName=dyn_cards_04_capsule, center=<5 0 0>, rotation=<rot 90 0 0>}]
 helper_thigh_l [{shapeType=Capsule, length=10.5137, radius=7.6913, shapeName=thigh_l_capsule, center=<-7.1372 1.002 0.3316>, rotation=<rot 90 0 0>}]
 pelvis [{shapeType=Capsule, length=7.1661, radius=9.7302, shapeName=pelvis_capsule}]
}
constraints [
 {
  bone1Name=dyn_belt_ref
  bone2Name=base_belt_ref
  swing1LimitDegrees=45
  swing1Motion=Free
  swing2LimitDegrees=45
  swing2Motion=Free
  twistLimitDegrees=45
  twistMotion=Free
 }
 {
  bone1Name=base_belt_cardsholder
  bone2Name=dyn_belt_ref
  swing1LimitDegrees=45
  swing1Motion=Free
  swing2LimitDegrees=45
  swing2Motion=Free
  twistLimitDegrees=45
  twistMotion=Free
 }
 {bone1Name=dyn_belt_cardsholder, bone2Name=base_belt_cardsholder, swing1LimitDegrees=10, swing2LimitDegrees=5, twistLimitDegrees=0}
 {bone1Name=dyn_cards_01, bone2Name=dyn_belt_cardsholder, swing1LimitDegrees=0, swing2LimitDegrees=0, twistLimitDegrees=0}
 {bone1Name=dyn_cards_02, bone2Name=dyn_belt_cardsholder, swing1LimitDegrees=0, swing2LimitDegrees=7, twistLimitDegrees=0}
 {bone1Name=dyn_cards_03, bone2Name=dyn_belt_cardsholder, swing1LimitDegrees=0, swing2LimitDegrees=7, twistLimitDegrees=0}
 {
  bone1Name=base_belt_ref
  bone2Name=pelvis
  swing1LimitDegrees=45
  swing1Motion=Free
  swing2LimitDegrees=45
  swing2Motion=Free
  twistLimitDegrees=45
  twistMotion=Free
 }
 {
  bone1Name=helper_thigh_l
  bone2Name=pelvis
  swing1LimitDegrees=45
  swing1Motion=Free
  swing2LimitDegrees=45
  swing2Motion=Free
  twistLimitDegrees=45
  twistMotion=Free
 }
 {
  bone1Name=dyn_belt_side_02
  bone2Name=dyn_belt_side_01
  swing1LimitDegrees=45
  swing1Motion=Free
  swing2LimitDegrees=45
  swing2Motion=Free
  twistLimitDegrees=45
  twistMotion=Free
 }
 {
  bone1Name=dyn_belt_side_01
  bone2Name=dyn_belt_ref
  swing1LimitDegrees=45
  swing1Motion=Free
  swing2LimitDegrees=45
  swing2Motion=Free
  twistLimitDegrees=45
  twistMotion=Free
 }
 {bone1Name=dyn_cards_04, bone2Name=dyn_belt_cardsholder, swing1LimitDegrees=0, swing2LimitDegrees=10, twistLimitDegrees=0}
]
