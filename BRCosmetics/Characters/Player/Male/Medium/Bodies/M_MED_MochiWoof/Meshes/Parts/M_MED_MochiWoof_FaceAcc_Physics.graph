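# Physics asset for the MochiWoof male medium body's FaceAcc mesh part, rigging a dangling belt with capsule bodies on its bones.
# Front and side belt chains branch from a shared reference bone linked to the pelvis, and their tips are constrained to each other.
# Upper joints rotate freely, while the last two joints of each chain have narrow swing limits and zero twist.
# Larger pelvis and left-thigh helper capsules approximate the body around the belt.
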PhysicsAsset /BRCosmetics/Characters/Player/Male/Medium/Bodies/M_MED_MochiWoof/Meshes/Parts/M_MED_MochiWoof_FaceAcc_Physics
bodies {
 base_belt_fr_00 [{shapeType=Capsule, length=0.505, radius=0.505, shapeName=base_belt_fr_00_capsule, rotation=<rot 0 -90 0>}]
 base_belt_ref [{shapeType=Capsule, length=0.505, radius=0.505, shapeName=base_belt_ref_capsule, rotation=<rot 0 -90 0>}]
 base_belt_side_00 [{shapeType=Capsule, length=0.505, radius=0.505, shapeName=base_belt_side_00_capsule, rotation=<rot 0 -90 0>}]
 dyn_belt_fr_00 [{shapeType=Capsule, length=0.505, radius=0.505, shapeName=dyn_belt_fr_00_capsule, rotation=<rot 0 -90 0>}]
 dyn_belt_fr_01 [{
   shapeType=Capsule
   length=0.9019
   radius=0.7567
   shapeName=dyn_belt_fr_01_capsule
   center=<2.725 0.1101 -0.0822>
   rotation=<rot 0 -14.9038 -79.6244>
  }]
 dyn_belt_fr_02 [{
   shapeType=Capsule
   length=1.3044
   radius=0.8117
   shapeName=dyn_belt_fr_02_capsule
   center=<1.2815 0.3423 -0.0022>
   rotation=<rot 0 98.8187 -89.7274>
  }]
 dyn_belt_fr_03 [{
   shapeType=Capsule
   length=4.8919
   radius=1.5977
   shapeName=dyn_belt_fr_03_capsule
   center=<2.6774 0 -0.6817>
   rotation=<rot 90 -101.4823 -81.4825>
  }]
 dyn_belt_ref [{shapeType=Capsule, length=0.505, radius=0.505, shapeName=dyn_belt_ref_capsule, rotation=<rot 0 -90 0>}]
 dyn_belt_side_00 [{shapeType=Capsule, length=0.505, radius=0.505, shapeName=dyn_belt_side_00_capsule, rotation=<rot 0 -90 0>}]
 dyn_belt_side_01 [{
   shapeType=Capsule
   length=1.0707
   radius=0.9743
   shapeName=dyn_belt_side_01_capsule
   center=<3.8043 -0.1824 -0.284>
   rotation=<rot 0 124.3618 -88.3418>
  }]
 dyn_belt_side_02 [{shapeType=Capsule, length=4.0533, radius=1.3223, shapeName=dyn_belt_side_02_capsule, center=<2.5927 0 0>, rotation=<rot 90 0 0>}]
 dyn_belt_side_03 [{
   shapeType=Capsule
   length=3.919
   radius=1.3577
   shapeName=dyn_belt_side_03_capsule
   center=<0.9856 0 -0.4948>
   rotation=<rot 75 0.0004 0.0003>
  }]
 helper_thigh_l [{shapeType=Capsule, length=10.7, radius=7.505, shapeName=thigh_l_capsule, center=<-7.1372 1.002 0.6485>, rotation=<rot 90 0 0>}]
 pelvis [{shapeType=Capsule, length=7.1661, radius=9.7302, shapeName=pelvis_capsule}]
}
constraints [
 {
  bone1Name=dyn_belt_ref
  bone2Name=base_belt_ref
  swing1LimitDegrees=45
  swing1Motion=Free
  swing2LimitDegrees=45
  swing2Motion=Free
  twistLimitDegrees=45
  twistMotion=Free
 }
 {
  bone1Name=base_belt_fr_00
  bone2Name=dyn_belt_ref
  swing1LimitDegrees=45
  swing1Motion=Free
  swing2LimitDegrees=45
  swing2Motion=Free
  twistLimitDegrees=45
  twistMotion=Free
 }
 {
  bone1Name=dyn_belt_fr_00
  bone2Name=base_belt_fr_00
  swing1LimitDegrees=45
  swing1Motion=Free
  swing2LimitDegrees=45
  swing2Motion=Free
  twistLimitDegrees=45
  twistMotion=Free
 }
 {
  bone1Name=dyn_belt_fr_01
  bone2Name=dyn_belt_fr_00
  swing1LimitDegrees=45
  swing1Motion=Free
  swing2LimitDegrees=45
  swing2Motion=Free
  twistLimitDegrees=45
  twistMotion=Free
 }
 {bone1Name=dyn_belt_fr_02, bone2Name=dyn_belt_fr_01, swing1LimitDegrees=10, swing2LimitDegrees=20, twistLimitDegrees=0}
 {bone1Name=dyn_belt_fr_03, bone2Name=dyn_belt_fr_02, swing1LimitDegrees=10, swing2LimitDegrees=25, twistLimitDegrees=0}
 {
  bone1Name=base_belt_side_00
  bone2Name=dyn_belt_ref
  swing1LimitDegrees=45
  swing1Motion=Free
  swing2LimitDegrees=45
  swing2Motion=Free
  twistLimitDegrees=45
  twistMotion=Free
 }
 {
  bone1Name=dyn_belt_side_00
  bone2Name=base_belt_side_00
  swing1LimitDegrees=45
  swing1Motion=Free
  swing2LimitDegrees=45
  swing2Motion=Free
  twistLimitDegrees=45
  twistMotion=Free
 }
 {
  bone1Name=dyn_belt_side_01
  bone2Name=dyn_belt_side_00
  swing1LimitDegrees=45
  swing1Motion=Free
  swing2LimitDegrees=45
  swing2Motion=Free
  twistLimitDegrees=45
  twistMotion=Free
 }
 {bone1Name=dyn_belt_side_02, bone2Name=dyn_belt_side_01, swing1LimitDegrees=10, swing2LimitDegrees=25, twistLimitDegrees=0}
 {bone1Name=dyn_belt_side_03, bone2Name=dyn_belt_side_02, swing1LimitDegrees=10, swing2LimitDegrees=25, twistLimitDegrees=0}
 {
  bone1Name=base_belt_ref
  bone2Name=pelvis
  swing1LimitDegrees=45
  swing1Motion=Free
  swing2LimitDegrees=45
  swing2Motion=Free
  twistLimitDegrees=45
  twistMotion=Free
 }
 {
  bone1Name=dyn_belt_side_03
  bone2Name=dyn_belt_fr_03
  swing1LimitDegrees=45
  swing1Motion=Free
  swing2LimitDegrees=45
  swing2Motion=Free
  twistLimitDegrees=45
  twistMotion=Free
 }
 {
  bone1Name=helper_thigh_l
  bone2Name=pelvis
  swing1LimitDegrees=45
  swing1Motion=Free
  swing2LimitDegrees=45
  swing2Motion=Free
  twistLimitDegrees=45
  twistMotion=Free
 }
]
